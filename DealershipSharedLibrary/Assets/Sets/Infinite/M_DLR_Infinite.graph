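# Translucent unlit FX material for Niagara mesh particles in the Infinite dealership set, previewed on the Infinite boost mesh.
# Panning caustics, tail gradient and tileable noise blend into a contrast-adjusted pattern, tinted by particle color for emissive adjustment.
# Particle alpha scales the pattern for opacity, optionally dissolved by a sine and U-gradient mask.
# Dynamic parameters supply fades and optional scroll offset; the FadeIn and FadeOut reroute names are swapped relative to their outputs.
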Material /DealershipSharedLibrary/Assets/Sets/Infinite/M_DLR_Infinite
graph {
 expressions {
  MaterialExpressionAdd_0 {
   class=MaterialExpressionAdd
   inputs [
    A {from=MaterialExpressionMultiply_1}
    B {from=MaterialExpressionAdd_1}
   ]
  }
  MaterialExpressionAdd_1 {
   class=MaterialExpressionAdd
   inputs [
    A {from=MaterialExpressionTextureSample_1, output=R}
    B {from=MaterialExpressionTextureSample_2, output=R}
   ]
  }
  MaterialExpressionAdd_2 {
   class=MaterialExpressionAdd
   inputs [
    A {from=MaterialExpressionNamedRerouteUsage_0}
    B {from=MaterialExpressionComponentMask_0}
   ]
  }
  MaterialExpressionAdd_3 {
   class=MaterialExpressionAdd
   inputs [
    A {from=MaterialExpressionMultiply_2}
    B {from=MaterialExpressionNamedRerouteUsage_1}
   ]
  }
  MaterialExpressionAdd_4 {
   class=MaterialExpressionAdd
   inputs [
    A {from=MaterialExpressionTextureCoordinate_0}
    B {from=MaterialExpressionMultiply_7}
   ]
  }
  MaterialExpressionComponentMask_0 {class=MaterialExpressionComponentMask, inputs [None {from=MaterialExpressionTextureCoordinate_3}], properties {r=true}}
  MaterialExpressionComponentMask_1 {class=MaterialExpressionComponentMask, inputs [None {from=MaterialExpressionTextureCoordinate_4}], properties {r=true}}
  MaterialExpressionConstantClamp_0 {class=MaterialExpressionSaturate, inputs [None {from=MaterialExpressionLinearInterpolate_0}]}
  MaterialExpressionConstant_0 {class=MaterialExpressionConstant, properties {r=5}}
  MaterialExpressionConvert_2 {
   class=MaterialExpressionConvert
   inputs [
    None {from=MaterialExpressionScalarParameter_5}
    None {from=MaterialExpressionScalarParameter_6}
   ]
   properties {
    nodeName=Make Vector2
    convertInputs [
     {
      type=Scalar
      expressionInput {
       expression=&self:MaterialExpressionScalarParameter_5
       inputName=None
       mask=0
       maskA=0
       maskB=0
       maskG=0
       maskR=0
       outputIndex=0
      }
     }
     {
      type=Scalar
      expressionInput {
       expression=&self:MaterialExpressionScalarParameter_6
       inputName=None
       mask=0
       maskA=0
       maskB=0
       maskG=0
       maskR=0
       outputIndex=0
      }
     }
    ]
    convertOutputs [{type=Vector2}]
   }
  }
  MaterialExpressionConvert_3 {
   class=MaterialExpressionConvert
   inputs [
    None {from=MaterialExpressionScalarParameter_7}
    None {from=MaterialExpressionScalarParameter_8}
   ]
   properties {
    nodeName=Make Vector2
    convertInputs [
     {
      type=Scalar
      expressionInput {
       expression=&self:MaterialExpressionScalarParameter_7
       inputName=None
       mask=0
       maskA=0
       maskB=0
       maskG=0
       maskR=0
       outputIndex=0
      }
     }
     {
      type=Scalar
      expressionInput {
       expression=&self:MaterialExpressionScalarParameter_8
       inputName=None
       mask=0
       maskA=0
       maskB=0
       maskG=0
       maskR=0
       outputIndex=0
      }
     }
    ]
    convertOutputs [{type=Vector2}]
   }
  }
  MaterialExpressionDynamicParameter_0 {
   class=MaterialExpressionDynamicParameter
   inputs []
   properties {defaultValue=<rgba -0.5 1 1 1>, paramNames [FadeIn, FadeOut, SpeedX, MultiplierSpeed]}
  }
  MaterialExpressionLinearInterpolate_0 {
   class=MaterialExpressionLinearInterpolate
   inputs [
    A {from=MaterialExpressionMultiply_3}
    B {from=MaterialExpressionMultiply_6}
    Alpha {from=MaterialExpressionScalarParameter_0}
   ]
  }
  MaterialExpressionLinearInterpolate_1 {
   class=MaterialExpressionLinearInterpolate
   inputs [
    A {from=MaterialExpressionAdd_0}
    B {from=MaterialExpressionReroute_3}
    Alpha {from=MaterialExpressionScalarParameter_1}
   ]
  }
  MaterialExpressionLinearInterpolate_2 {
   class=MaterialExpressionLinearInterpolate
   inputs [
    A {from=MaterialExpressionPanner_0}
    B {from=MaterialExpressionAdd_4}
    Alpha {from=MaterialExpressionScalarParameter_2}
   ]
  }
  MaterialExpressionLinearInterpolate_3 {
   class=MaterialExpressionLinearInterpolate
   inputs [
    A {from=MaterialExpressionPower_1}
    B {from=MaterialExpressionSaturate_0}
    Alpha {from=MaterialExpressionScalarParameter_4}
   ]
  }
  MaterialExpressionMaterialFunctionCall_0 {
   class=MaterialExpressionMaterialFunctionCall
   inputs [Emissive {from=MaterialExpressionMultiply_0}]
   properties {materialFunction=&MF_FX_EmissiveAdjustments}
  }
  MaterialExpressionMaterialFunctionCall_1 {
   class=MaterialExpressionMaterialFunctionCall
   inputs [
    X {from=MaterialExpressionNamedRerouteUsage_2}
    Y {from=""}
   ]
   properties {materialFunction=&MakeFloat2}
  }
  MaterialExpressionMeshEmitterVertexColor_0 {class=MaterialExpressionParticleColor, inputs [], properties {}}
  MaterialExpressionMultiply_0 {
   class=MaterialExpressionMultiply
   inputs [
    A {from=MaterialExpressionMeshEmitterVertexColor_0, output=RGB}
    B {from=MaterialExpressionLinearInterpolate_3}
   ]
  }
  MaterialExpressionMultiply_1 {
   class=MaterialExpressionMultiply
   inputs [
    A {from=MaterialExpressionTextureSample_0, output=R}
    B {from=MaterialExpressionConstant_0}
   ]
  }
  MaterialExpressionMultiply_2 {
   class=MaterialExpressionMultiply
   inputs [
    A {from=MaterialExpressionPower_0}
    B {from=""}
   ]
   properties {constB=5}
  }
  MaterialExpressionMultiply_3 {
   class=MaterialExpressionMultiply
   inputs [
    A {from=MaterialExpressionMeshEmitterVertexColor_0, output=A}
    B {from=MaterialExpressionSaturate_0}
   ]
  }
  MaterialExpressionMultiply_4 {
   class=MaterialExpressionMultiply
   inputs [
    A {from=MaterialExpressionReroute_1}
    B {from=MaterialExpressionAdd_3}
   ]
  }
  MaterialExpressionMultiply_5 {
   class=MaterialExpressionMultiply
   inputs [
    A {from=MaterialExpressionAdd_2}
    B {from=MaterialExpressionAdd_3}
   ]
  }
  MaterialExpressionMultiply_6 {
   class=MaterialExpressionMultiply
   inputs [
    A {from=MaterialExpressionMultiply_3}
    B {from=MaterialExpressionMultiply_8}
   ]
  }
  MaterialExpressionMultiply_7 {
   class=MaterialExpressionMultiply
   inputs [
    A {from=MaterialExpressionMaterialFunctionCall_1, output=Result}
    B {from=MaterialExpressionNamedRerouteUsage_3}
   ]
  }
  MaterialExpressionMultiply_8 {
   class=MaterialExpressionMultiply
   inputs [
    A {from=MaterialExpressionPower_2}
    B {from=MaterialExpressionSaturate_1}
   ]
  }
  MaterialExpressionMultiply_9 {
   class=MaterialExpressionMultiply
   inputs [
    A {from=MaterialExpressionTextureCoordinate_2}
    B {from=MaterialExpressionConvert_2}
   ]
  }
  MaterialExpressionNamedRerouteDeclaration_0 {
   class=MaterialExpressionNamedRerouteDeclaration
   inputs [None {from=MaterialExpressionDynamicParameter_0, output=FadeIn}]
   properties {name=FadeOut, nodeColor=<rgba 0.5938 0 1 1>}
  }
  MaterialExpressionNamedRerouteDeclaration_1 {
   class=MaterialExpressionNamedRerouteDeclaration
   inputs [None {from=MaterialExpressionDynamicParameter_0, output=FadeOut}]
   properties {name=FadeIn, nodeColor=<rgba 0.5938 0 1 1>}
  }
  MaterialExpressionNamedRerouteDeclaration_2 {
   class=MaterialExpressionNamedRerouteDeclaration
   inputs [None {from=MaterialExpressionDynamicParameter_0, output=SpeedX}]
   properties {name=SpeedX, nodeColor=<rgba 0 1 0.2734 1>}
  }
  MaterialExpressionNamedRerouteDeclaration_3 {
   class=MaterialExpressionNamedRerouteDeclaration
   inputs [None {from=MaterialExpressionDynamicParameter_0, output=MultiplierSpeed}]
   properties {name=MultiplierSpeed, nodeColor=<rgba 0 1 0.2734 1>}
  }
  MaterialExpressionNamedRerouteUsage_0 {class=MaterialExpressionNamedRerouteUsage}
  MaterialExpressionNamedRerouteUsage_1 {class=MaterialExpressionNamedRerouteUsage}
  MaterialExpressionNamedRerouteUsage_2 {class=MaterialExpressionNamedRerouteUsage}
  MaterialExpressionNamedRerouteUsage_3 {class=MaterialExpressionNamedRerouteUsage}
  MaterialExpressionOneMinus_0 {class=MaterialExpressionOneMinus, inputs [None {from=MaterialExpressionComponentMask_0}]}
  MaterialExpressionPanner_0 {
   class=MaterialExpressionPanner
   inputs [
    Coordinate {from=MaterialExpressionTextureCoordinate_0}
    Time {from=""}
    Speed {from=""}
   ]
   properties {speedX=-0.7}
  }
  MaterialExpressionPanner_1 {
   class=MaterialExpressionPanner
   inputs [
    Coordinate {from=MaterialExpressionTextureCoordinate_1}
    Time {from=""}
    Speed {from=""}
   ]
   properties {speedX=-1}
  }
  MaterialExpressionPanner_2 {
   class=MaterialExpressionPanner
   inputs [
    Coordinate {from=MaterialExpressionMultiply_9}
    Time {from=""}
    Speed {from=MaterialExpressionConvert_3}
   ]
   properties {speedX=0.2}
  }
  MaterialExpressionPower_0 {
   class=MaterialExpressionPower
   inputs [
    Base {from=MaterialExpressionOneMinus_0}
    Exp {from=""}
   ]
   properties {constExponent=1.5}
  }
  MaterialExpressionPower_1 {
   class=MaterialExpressionPower
   inputs [
    Base {from=MaterialExpressionLinearInterpolate_1}
    Exp {from=MaterialExpressionScalarParameter_3}
   ]
  }
  MaterialExpressionPower_2 {
   class=MaterialExpressionPower
   inputs [
    Base {from=MaterialExpressionSine_0}
    Exp {from=""}
   ]
  }
  MaterialExpressionReroute_1 {class=MaterialExpressionReroute, inputs [None {from=MaterialExpressionComponentMask_0}]}
  MaterialExpressionReroute_2 {class=MaterialExpressionReroute, inputs [None {from=MaterialExpressionTextureSample_2, output=R}]}
  MaterialExpressionReroute_3 {class=MaterialExpressionReroute, inputs [None {from=MaterialExpressionReroute_2}]}
  MaterialExpressionReroute_4 {class=MaterialExpressionReroute, inputs [None {from=MaterialExpressionDynamicParameter_0, output=SpeedX}]}
  MaterialExpressionSaturate_0 {class=MaterialExpressionSaturate, inputs [None {from=MaterialExpressionPower_1}]}
  MaterialExpressionSaturate_1 {class=MaterialExpressionSaturate, inputs [None {from=MaterialExpressionMultiply_5}]}
  MaterialExpressionScalarParameter_0 {
   class=MaterialExpressionScalarParameter
   properties {controlType=Enumeration, enumeration=/DealershipSharedLibrary/FX/Materials/Enums/Enum_DLR_TrueFalse, parameterName=UseDissolve?}
  }
  MaterialExpressionScalarParameter_1 {
   class=MaterialExpressionScalarParameter
   properties {controlType=Enumeration, enumeration=&Enum_DLR_TrueFalse, parameterName=UseNoiseTextureOnly?}
  }
  MaterialExpressionScalarParameter_2 {
   class=MaterialExpressionScalarParameter
   properties {controlType=Enumeration, enumeration=&Enum_DLR_TrueFalse, parameterName=UseSpeedFromDynamic?}
  }
  MaterialExpressionScalarParameter_3 {class=MaterialExpressionScalarParameter, properties {defaultValue=1, parameterName=TextureContrast}}
  MaterialExpressionScalarParameter_4 {
   class=MaterialExpressionScalarParameter
   properties {controlType=Enumeration, enumeration=&Enum_DLR_TrueFalse, parameterName=SaturateTextureConstrast?}
  }
  MaterialExpressionScalarParameter_5 {class=MaterialExpressionScalarParameter, properties {defaultValue=5, parameterName=NoiseTilingX}}
  MaterialExpressionScalarParameter_6 {class=MaterialExpressionScalarParameter, properties {defaultValue=1, parameterName=NoiseTilingY}}
  MaterialExpressionScalarParameter_7 {class=MaterialExpressionScalarParameter, properties {defaultValue=-3, parameterName=NoiseSpeedX}}
  MaterialExpressionScalarParameter_8 {class=MaterialExpressionScalarParameter, properties {parameterName=NoiseSpeedY}}
  MaterialExpressionSine_0 {class=MaterialExpressionSine, inputs [None {from=MaterialExpressionComponentMask_1}], properties {period=2}}
  MaterialExpressionTextureCoordinate_0 {class=MaterialExpressionTextureCoordinate}
  MaterialExpressionTextureCoordinate_1 {class=MaterialExpressionTextureCoordinate, properties {vTiling=2}}
  MaterialExpressionTextureCoordinate_2 {class=MaterialExpressionTextureCoordinate}
  MaterialExpressionTextureCoordinate_3 {class=MaterialExpressionTextureCoordinate}
  MaterialExpressionTextureCoordinate_4 {class=MaterialExpressionTextureCoordinate}
  MaterialExpressionTextureSample_0 {
   class=MaterialExpressionTextureSample
   inputs [
    UVs {from=MaterialExpressionLinearInterpolate_2}
    Tex {from=""}
    Apply View MipBias {from=""}
   ]
   properties {texture=/DealershipSharedLibrary/FX/Textures/Noise/T_VCVFX_CausticsMask_RGB}
  }
  MaterialExpressionTextureSample_1 {
   class=MaterialExpressionTextureSample
   inputs [
    UVs {from=MaterialExpressionPanner_1}
    Tex {from=""}
    Apply View MipBias {from=""}
   ]
   properties {texture=/DealershipSharedLibrary/FX/Textures/Gradient_Masks/T_VCVFX_DarkcloudTail_A}
  }
  MaterialExpressionTextureSample_2 {
   class=MaterialExpressionTextureSample
   inputs [
    UVs {from=MaterialExpressionPanner_2}
    Tex {from=""}
    Apply View MipBias {from=""}
   ]
   properties {texture=/DealershipSharedLibrary/FX/Textures/Noise/T_VCVFX_NoiseCones_A}
  }
 }
 properties {
  bAutomaticallySetUsageInEditor=false
  bEnableResponsiveAA=true
  bUsedWithNiagaraMeshParticles=true
  blendMode=BLEND_Translucent
  previewMesh=/VehicleCosmetics/Boosts/Boost_Infinite/Meshes/SM_Infinite
  shadingModel=MSM_Unlit
  usedShadingModels=Unlit
 }
}
