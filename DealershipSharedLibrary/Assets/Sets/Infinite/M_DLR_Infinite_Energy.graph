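# Unlit, masked scrolling energy material for Niagara mesh particles in the Dealership Infinite set.
# A noise-distorted main texture gains an optional secondary layer with sine-wave strips, plus an optional panning clouds texture.
# Particle color tints this for emissive adjustments, while particle alpha and an optional, separately distorted mask shape opacity through saturate and depth fade.
# Feature toggles are enum-driven scalar lerps rather than static switches.
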Material /DealershipSharedLibrary/Assets/Sets/Infinite/M_DLR_Infinite_Energy
graph {
 expressions {
  MaterialExpressionAdd_0 {
   class=MaterialExpressionAdd
   inputs [
    A {from=MaterialExpressionMultiply_3}
    B {from=MaterialExpressionMultiply_2}
   ]
  }
  MaterialExpressionAdd_1 {
   class=MaterialExpressionAdd
   inputs [
    A {from=MaterialExpressionLinearInterpolate_1}
    B {from=MaterialExpressionNamedRerouteUsage_15}
   ]
  }
  MaterialExpressionAdd_2 {
   class=MaterialExpressionAdd
   inputs [
    A {from=MaterialExpressionMultiply_4}
    B {from=MaterialExpressionLinearInterpolate_0}
   ]
  }
  MaterialExpressionAdd_3 {
   class=MaterialExpressionAdd
   inputs [
    A {from=MaterialExpressionMultiply_9}
    B {from=MaterialExpressionMultiply_8}
   ]
  }
  MaterialExpressionComponentMask_0 {
   class=MaterialExpressionComponentMask
   inputs [None {from=MaterialExpressionTextureSampleParameter2D_0, output=RGB}]
   properties {g=true, r=true}
  }
  MaterialExpressionComponentMask_1 {class=MaterialExpressionComponentMask, inputs [None {from=MaterialExpressionPanner_2}], properties {g=true}}
  MaterialExpressionComponentMask_2 {
   class=MaterialExpressionComponentMask
   inputs [None {from=MaterialExpressionTextureSampleParameter2D_2, output=RGB}]
   properties {g=true, r=true}
  }
  MaterialExpressionConvert_0 {
   class=MaterialExpressionConvert
   inputs [
    None {from=MaterialExpressionScalarParameter_15}
    None {from=MaterialExpressionScalarParameter_16}
   ]
   properties {
    nodeName=Make Vector2
    convertInputs [
     {
      type=Scalar
      expressionInput {
       expression=&self:MaterialExpressionScalarParameter_15
       inputName=None
       mask=0
       maskA=0
       maskB=0
       maskG=0
       maskR=0
       outputIndex=0
      }
     }
     {
      type=Scalar
      expressionInput {
       expression=&self:MaterialExpressionScalarParameter_16
       inputName=None
       mask=0
       maskA=0
       maskB=0
       maskG=0
       maskR=0
       outputIndex=0
      }
     }
    ]
    convertOutputs [{type=Vector2}]
   }
  }
  MaterialExpressionConvert_2 {
   class=MaterialExpressionConvert
   inputs [
    None {from=MaterialExpressionScalarParameter_17}
    None {from=MaterialExpressionScalarParameter_18}
   ]
   properties {
    nodeName=Make Vector2
    convertInputs [
     {
      type=Scalar
      expressionInput {
       expression=&self:MaterialExpressionScalarParameter_17
       inputName=None
       mask=0
       maskA=0
       maskB=0
       maskG=0
       maskR=0
       outputIndex=0
      }
     }
     {
      type=Scalar
      expressionInput {
       expression=&self:MaterialExpressionScalarParameter_18
       inputName=None
       mask=0
       maskA=0
       maskB=0
       maskG=0
       maskR=0
       outputIndex=0
      }
     }
    ]
    convertOutputs [{type=Vector2}]
   }
  }
  MaterialExpressionDivide_1 {
   class=MaterialExpressionDivide
   inputs [
    A {from=MaterialExpressionSine_0}
    B {from=MaterialExpressionScalarParameter_19}
   ]
   properties {constB=1.5}
  }
  MaterialExpressionLinearInterpolate_0 {
   class=MaterialExpressionLinearInterpolate
   inputs [
    A {from=MaterialExpressionMultiply_5}
    B {from=MaterialExpressionMultiply_6}
    Alpha {from=MaterialExpressionSaturate_0}
   ]
  }
  MaterialExpressionLinearInterpolate_1 {
   class=MaterialExpressionLinearInterpolate
   inputs [
    A {from=MaterialExpressionReroute_4}
    B {from=MaterialExpressionAdd_2}
    Alpha {from=MaterialExpressionSaturate_1}
   ]
  }
  MaterialExpressionLinearInterpolate_2 {
   class=MaterialExpressionLinearInterpolate
   inputs [
    A {from=""}
    B {from=MaterialExpressionMultiply_0}
    Alpha {from=MaterialExpressionSaturate_2}
   ]
   properties {constA=1}
  }
  MaterialExpressionLinearInterpolate_3 {
   class=MaterialExpressionLinearInterpolate
   inputs [
    A {from=""}
    B {from=MaterialExpressionMultiply_16}
    Alpha {from=MaterialExpressionScalarParameter_11}
   ]
  }
  MaterialExpressionMaterialFunctionCall_0 {
   class=MaterialExpressionMaterialFunctionCall
   inputs [
    X {from=MaterialExpressionVectorParameter_1, output=Tiling X}
    Y {from=MaterialExpressionVectorParameter_1, output=Tiling X}
   ]
   properties {materialFunction=&MakeFloat2}
  }
  MaterialExpressionMaterialFunctionCall_1 {
   class=MaterialExpressionMaterialFunctionCall
   inputs [
    X {from=MaterialExpressionVectorParameter_0, output=Tiling X}
    Y {from=MaterialExpressionVectorParameter_0, output=Tiling X}
   ]
   properties {materialFunction=&MakeFloat2}
  }
  MaterialExpressionMaterialFunctionCall_11 {
   class=MaterialExpressionMaterialFunctionCall
   inputs [Emissive {from=MaterialExpressionPower_5}]
   properties {materialFunction=&MF_FX_EmissiveAdjustments}
  }
  MaterialExpressionMaterialFunctionCall_12 {
   class=MaterialExpressionMaterialFunctionCall
   inputs [
    Opacity In {from=MaterialExpressionSaturate_3}
    DepthFade Distance {from=""}
   ]
   properties {materialFunction=&MF_VCVFX_DepthFade}
  }
  MaterialExpressionMaterialFunctionCall_2 {
   class=MaterialExpressionMaterialFunctionCall
   inputs [
    X {from=MaterialExpressionVectorParameter_0, output=Speed X}
    Y {from=MaterialExpressionVectorParameter_0, output=Speed X}
   ]
   properties {materialFunction=&MakeFloat2}
  }
  MaterialExpressionMaterialFunctionCall_3 {
   class=MaterialExpressionMaterialFunctionCall
   inputs [
    X {from=MaterialExpressionVectorParameter_1, output=Speed X}
    Y {from=MaterialExpressionVectorParameter_1, output=Speed X}
   ]
   properties {materialFunction=&MakeFloat2}
  }
  MaterialExpressionMaterialFunctionCall_4 {
   class=MaterialExpressionMaterialFunctionCall
   inputs [
    X {from=MaterialExpressionVectorParameter_2, output=Tiling X}
    Y {from=MaterialExpressionVectorParameter_2, output=Tiling X}
   ]
   properties {materialFunction=&MakeFloat2}
  }
  MaterialExpressionMaterialFunctionCall_5 {
   class=MaterialExpressionMaterialFunctionCall
   inputs [
    X {from=MaterialExpressionVectorParameter_2, output=Speed X}
    Y {from=MaterialExpressionVectorParameter_2, output=Speed X}
   ]
   properties {materialFunction=&MakeFloat2}
  }
  MaterialExpressionMaterialFunctionCall_6 {
   class=MaterialExpressionMaterialFunctionCall
   inputs [
    X {from=MaterialExpressionVectorParameter_3, output=Tiling X}
    Y {from=MaterialExpressionVectorParameter_3, output=Tiling X}
   ]
   properties {materialFunction=&MakeFloat2}
  }
  MaterialExpressionMaterialFunctionCall_7 {
   class=MaterialExpressionMaterialFunctionCall
   inputs [
    X {from=MaterialExpressionVectorParameter_3, output=Speed X}
    Y {from=MaterialExpressionVectorParameter_3, output=Speed X}
   ]
   properties {materialFunction=&MakeFloat2}
  }
  MaterialExpressionMaterialFunctionCall_8 {
   class=MaterialExpressionMaterialFunctionCall
   inputs [
    X {from=MaterialExpressionVectorParameter_4, output=Tiling X}
    Y {from=MaterialExpressionVectorParameter_4, output=Tiling X}
   ]
   properties {materialFunction=&MakeFloat2}
  }
  MaterialExpressionMaterialFunctionCall_9 {
   class=MaterialExpressionMaterialFunctionCall
   inputs [
    X {from=MaterialExpressionVectorParameter_4, output=Speed X}
    Y {from=MaterialExpressionVectorParameter_4, output=Speed X}
   ]
   properties {materialFunction=&MakeFloat2}
  }
  MaterialExpressionMultiply_0 {
   class=MaterialExpressionMultiply
   inputs [
    A {from=MaterialExpressionPower_3}
    B {from=MaterialExpressionScalarParameter_9}
   ]
  }
  MaterialExpressionMultiply_1 {
   class=MaterialExpressionMultiply
   inputs [
    A {from=MaterialExpressionTextureCoordinate_0}
    B {from=MaterialExpressionNamedRerouteUsage_3}
   ]
  }
  MaterialExpressionMultiply_10 {
   class=MaterialExpressionMultiply
   inputs [
    A {from=MaterialExpressionTextureCoordinate_2}
    B {from=MaterialExpressionConvert_0}
   ]
  }
  MaterialExpressionMultiply_11 {
   class=MaterialExpressionMultiply
   inputs [
    A {from=MaterialExpressionTextureCoordinate_5}
    B {from=MaterialExpressionNamedRerouteUsage_14}
   ]
  }
  MaterialExpressionMultiply_12 {
   class=MaterialExpressionMultiply
   inputs [
    A {from=MaterialExpressionAdd_1}
    B {from=MaterialExpressionParticleColor_0, output=A}
   ]
  }
  MaterialExpressionMultiply_13 {
   class=MaterialExpressionMultiply
   inputs [
    A {from=MaterialExpressionParticleColor_0, output=RGB}
    B {from=MaterialExpressionAdd_1}
   ]
  }
  MaterialExpressionMultiply_14 {
   class=MaterialExpressionMultiply
   inputs [
    A {from=MaterialExpressionMultiply_12}
    B {from=MaterialExpressionNamedRerouteUsage_16}
   ]
  }
  MaterialExpressionMultiply_16 {
   class=MaterialExpressionMultiply
   inputs [
    A {from=MaterialExpressionPower_4}
    B {from=MaterialExpressionScalarParameter_13}
   ]
  }
  MaterialExpressionMultiply_2 {
   class=MaterialExpressionMultiply
   inputs [
    A {from=MaterialExpressionComponentMask_0}
    B {from=MaterialExpressionScalarParameter_0}
   ]
  }
  MaterialExpressionMultiply_3 {
   class=MaterialExpressionMultiply
   inputs [
    A {from=MaterialExpressionTextureCoordinate_1}
    B {from=MaterialExpressionNamedRerouteUsage_4}
   ]
  }
  MaterialExpressionMultiply_4 {
   class=MaterialExpressionMultiply
   inputs [
    A {from=MaterialExpressionPower_1}
    B {from=MaterialExpressionScalarParameter_2}
   ]
  }
  MaterialExpressionMultiply_5 {
   class=MaterialExpressionMultiply
   inputs [
    A {from=MaterialExpressionPower_2}
    B {from=MaterialExpressionScalarParameter_4}
   ]
  }
  MaterialExpressionMultiply_6 {
   class=MaterialExpressionMultiply
   inputs [
    A {from=MaterialExpressionMultiply_5}
    B {from=MaterialExpressionPower_0}
   ]
  }
  MaterialExpressionMultiply_7 {
   class=MaterialExpressionMultiply
   inputs [
    A {from=MaterialExpressionTextureCoordinate_3}
    B {from=MaterialExpressionNamedRerouteUsage_8}
   ]
  }
  MaterialExpressionMultiply_8 {
   class=MaterialExpressionMultiply
   inputs [
    A {from=MaterialExpressionComponentMask_2}
    B {from=MaterialExpressionScalarParameter_7}
   ]
  }
  MaterialExpressionMultiply_9 {
   class=MaterialExpressionMultiply
   inputs [
    A {from=MaterialExpressionTextureCoordinate_4}
    B {from=MaterialExpressionNamedRerouteUsage_6}
   ]
  }
  MaterialExpressionNamedRerouteDeclaration_0 {
   class=MaterialExpressionNamedRerouteDeclaration
   inputs [None {from=MaterialExpressionMaterialFunctionCall_1, output=Result}]
   properties {name=MainTiling, nodeColor=<rgba 0 1 0.9063 1>}
  }
  MaterialExpressionNamedRerouteDeclaration_1 {
   class=MaterialExpressionNamedRerouteDeclaration
   inputs [None {from=MaterialExpressionMaterialFunctionCall_2, output=Result}]
   properties {name=MainSpeed, nodeColor=<rgba 1 0.5859 0 1>}
  }
  MaterialExpressionNamedRerouteDeclaration_10 {
   class=MaterialExpressionNamedRerouteDeclaration
   inputs [None {from=MaterialExpressionLinearInterpolate_3}]
   properties {name=CloudTexture, nodeColor=<rgba 0.5703 0 1 1>}
  }
  MaterialExpressionNamedRerouteDeclaration_11 {
   class=MaterialExpressionNamedRerouteDeclaration
   inputs [None {from=MaterialExpressionLinearInterpolate_2}]
   properties {name=OpcaityMask, nodeColor=<rgba 0 1 0.4844 1>}
  }
  MaterialExpressionNamedRerouteDeclaration_2 {
   class=MaterialExpressionNamedRerouteDeclaration
   inputs [None {from=MaterialExpressionMaterialFunctionCall_0, output=Result}]
   properties {name=DistortionTiling, nodeColor=<rgba 0 1 0.9062 1>}
  }
  MaterialExpressionNamedRerouteDeclaration_3 {
   class=MaterialExpressionNamedRerouteDeclaration
   inputs [None {from=MaterialExpressionMaterialFunctionCall_3, output=Result}]
   properties {name=DistortionSpeed, nodeColor=<rgba 1 0.5859 0 1>}
  }
  MaterialExpressionNamedRerouteDeclaration_4 {
   class=MaterialExpressionNamedRerouteDeclaration
   inputs [None {from=MaterialExpressionMaterialFunctionCall_4, output=Result}]
   properties {name=AlphaDistortionTiling, nodeColor=<rgba 0 1 0.9062 1>}
  }
  MaterialExpressionNamedRerouteDeclaration_5 {
   class=MaterialExpressionNamedRerouteDeclaration
   inputs [None {from=MaterialExpressionMaterialFunctionCall_5, output=Result}]
   properties {name=AlphaDistortionSpeed, nodeColor=<rgba 1 0.5859 0 1>}
  }
  MaterialExpressionNamedRerouteDeclaration_6 {
   class=MaterialExpressionNamedRerouteDeclaration
   inputs [None {from=MaterialExpressionMaterialFunctionCall_6, output=Result}]
   properties {name=AlphaTiling, nodeColor=<rgba 0 1 0.9062 1>}
  }
  MaterialExpressionNamedRerouteDeclaration_7 {
   class=MaterialExpressionNamedRerouteDeclaration
   inputs [None {from=MaterialExpressionMaterialFunctionCall_7, output=Result}]
   properties {name=AlphaSpeed, nodeColor=<rgba 1 0.5859 0 1>}
  }
  MaterialExpressionNamedRerouteDeclaration_8 {
   class=MaterialExpressionNamedRerouteDeclaration
   inputs [None {from=MaterialExpressionMaterialFunctionCall_8, output=Result}]
   properties {name=CloudsTiling, nodeColor=<rgba 0 1 0.9062 1>}
  }
  MaterialExpressionNamedRerouteDeclaration_9 {
   class=MaterialExpressionNamedRerouteDeclaration
   inputs [None {from=MaterialExpressionMaterialFunctionCall_9, output=Result}]
   properties {name=CloudsSpeed, nodeColor=<rgba 1 0.5859 0 1>}
  }
  MaterialExpressionNamedRerouteUsage_10 {class=MaterialExpressionNamedRerouteUsage}
  MaterialExpressionNamedRerouteUsage_13 {class=MaterialExpressionNamedRerouteUsage}
  MaterialExpressionNamedRerouteUsage_14 {class=MaterialExpressionNamedRerouteUsage}
  MaterialExpressionNamedRerouteUsage_15 {class=MaterialExpressionNamedRerouteUsage}
  MaterialExpressionNamedRerouteUsage_16 {class=MaterialExpressionNamedRerouteUsage}
  MaterialExpressionNamedRerouteUsage_2 {class=MaterialExpressionNamedRerouteUsage}
  MaterialExpressionNamedRerouteUsage_3 {class=MaterialExpressionNamedRerouteUsage}
  MaterialExpressionNamedRerouteUsage_4 {class=MaterialExpressionNamedRerouteUsage}
  MaterialExpressionNamedRerouteUsage_5 {class=MaterialExpressionNamedRerouteUsage}
  MaterialExpressionNamedRerouteUsage_6 {class=MaterialExpressionNamedRerouteUsage}
  MaterialExpressionNamedRerouteUsage_7 {class=MaterialExpressionNamedRerouteUsage}
  MaterialExpressionNamedRerouteUsage_8 {class=MaterialExpressionNamedRerouteUsage}
  MaterialExpressionPanner_0 {
   class=MaterialExpressionPanner
   inputs [
    Coordinate {from=MaterialExpressionMultiply_1}
    Time {from=""}
    Speed {from=MaterialExpressionNamedRerouteUsage_2}
   ]
  }
  MaterialExpressionPanner_1 {
   class=MaterialExpressionPanner
   inputs [
    Coordinate {from=MaterialExpressionAdd_0}
    Time {from=""}
    Speed {from=MaterialExpressionNamedRerouteUsage_5}
   ]
  }
  MaterialExpressionPanner_2 {
   class=MaterialExpressionPanner
   inputs [
    Coordinate {from=MaterialExpressionMultiply_10}
    Time {from=""}
    Speed {from=MaterialExpressionConvert_2}
   ]
   properties {speedY=2}
  }
  MaterialExpressionPanner_3 {
   class=MaterialExpressionPanner
   inputs [
    Coordinate {from=MaterialExpressionMultiply_7}
    Time {from=""}
    Speed {from=MaterialExpressionNamedRerouteUsage_10}
   ]
  }
  MaterialExpressionPanner_4 {
   class=MaterialExpressionPanner
   inputs [
    Coordinate {from=MaterialExpressionAdd_3}
    Time {from=""}
    Speed {from=MaterialExpressionNamedRerouteUsage_7}
   ]
  }
  MaterialExpressionPanner_5 {
   class=MaterialExpressionPanner
   inputs [
    Coordinate {from=MaterialExpressionMultiply_11}
    Time {from=""}
    Speed {from=MaterialExpressionNamedRerouteUsage_13}
   ]
  }
  MaterialExpressionParticleColor_0 {class=MaterialExpressionParticleColor, inputs [], properties {}}
  MaterialExpressionPower_0 {
   class=MaterialExpressionPower
   inputs [
    Base {from=MaterialExpressionDivide_1}
    Exp {from=MaterialExpressionScalarParameter_20}
   ]
   properties {constExponent=3}
  }
  MaterialExpressionPower_1 {
   class=MaterialExpressionPower
   inputs [
    Base {from=MaterialExpressionTextureSampleParameter2D_1, output=RGB}
    Exp {from=MaterialExpressionScalarParameter_1}
   ]
  }
  MaterialExpressionPower_2 {
   class=MaterialExpressionPower
   inputs [
    Base {from=MaterialExpressionTextureSampleParameter2D_1, output=RGB}
    Exp {from=MaterialExpressionScalarParameter_3}
   ]
  }
  MaterialExpressionPower_3 {
   class=MaterialExpressionPower
   inputs [
    Base {from=MaterialExpressionTextureSampleParameter2D_3, output=RGB}
    Exp {from=MaterialExpressionScalarParameter_8}
   ]
  }
  MaterialExpressionPower_4 {
   class=MaterialExpressionPower
   inputs [
    Base {from=MaterialExpressionTextureSampleParameter2D_4, output=RGB}
    Exp {from=MaterialExpressionScalarParameter_12}
   ]
  }
  MaterialExpressionPower_5 {
   class=MaterialExpressionPower
   inputs [
    Base {from=MaterialExpressionReroute_9}
    Exp {from=MaterialExpressionScalarParameter_14}
   ]
  }
  MaterialExpressionReroute_10 {class=MaterialExpressionReroute, inputs [None {from=MaterialExpressionSaturate_3}]}
  MaterialExpressionReroute_11 {class=MaterialExpressionReroute, inputs [None {from=MaterialExpressionReroute_10}]}
  MaterialExpressionReroute_4 {class=MaterialExpressionReroute, inputs [None {from=MaterialExpressionMultiply_4}]}
  MaterialExpressionReroute_9 {class=MaterialExpressionReroute, inputs [None {from=MaterialExpressionMultiply_13}]}
  MaterialExpressionSaturate_0 {class=MaterialExpressionSaturate, inputs [None {from=MaterialExpressionScalarParameter_5}]}
  MaterialExpressionSaturate_1 {class=MaterialExpressionSaturate, inputs [None {from=MaterialExpressionScalarParameter_6}]}
  MaterialExpressionSaturate_2 {class=MaterialExpressionSaturate, inputs [None {from=MaterialExpressionScalarParameter_10}]}
  MaterialExpressionSaturate_3 {class=MaterialExpressionSaturate, inputs [None {from=MaterialExpressionMultiply_14}]}
  MaterialExpressionScalarParameter_0 {class=MaterialExpressionScalarParameter, properties {defaultValue=0.08, group=Distortion, parameterName=Distortion Strength}}
  MaterialExpressionScalarParameter_1 {class=MaterialExpressionScalarParameter, properties {defaultValue=1, group=Main Texture, parameterName=Darkness, sortPriority=3}}
  MaterialExpressionScalarParameter_10 {
   class=MaterialExpressionScalarParameter
   properties {
    controlType=Enumeration
    enumeration=/DealershipSharedLibrary/FX/Materials/Enums/Enum_DLR_TrueFalse
    group=OpacityMask
    parameterName=UseAlphaMask?
    sortPriority=0
   }
  }
  MaterialExpressionScalarParameter_11 {
   class=MaterialExpressionScalarParameter
   properties {
    controlType=Enumeration
    enumeration=&Enum_DLR_TrueFalse
    group=Clouds Texture
    parameterName=UseCloudsTexture?
    sortPriority=0
   }
  }
  MaterialExpressionScalarParameter_12 {class=MaterialExpressionScalarParameter, properties {defaultValue=1, group=Clouds Texture, parameterName=CloudsDarkness, sortPriority=3}}
  MaterialExpressionScalarParameter_13 {class=MaterialExpressionScalarParameter, properties {defaultValue=1, group=Clouds Texture, parameterName=CloudsBrightness, sortPriority=4}}
  MaterialExpressionScalarParameter_14 {class=MaterialExpressionScalarParameter, properties {defaultValue=1, parameterName=Exp}}
  MaterialExpressionScalarParameter_15 {class=MaterialExpressionScalarParameter, properties {defaultValue=1, group=SecondaryMask, parameterName=StripMaskTilingX}}
  MaterialExpressionScalarParameter_16 {class=MaterialExpressionScalarParameter, properties {defaultValue=10, group=SecondaryMask, parameterName=StripMaskTilingY}}
  MaterialExpressionScalarParameter_17 {class=MaterialExpressionScalarParameter, properties {parameterName=StripMaskSpeedX}}
  MaterialExpressionScalarParameter_18 {class=MaterialExpressionScalarParameter, properties {defaultValue=2, group=SecondaryMask, parameterName=StripMaskSpeedY}}
  MaterialExpressionScalarParameter_19 {class=MaterialExpressionScalarParameter, properties {defaultValue=1.5, group=SecondaryMask, parameterName=StripMaskDivide}}
  MaterialExpressionScalarParameter_2 {class=MaterialExpressionScalarParameter, properties {defaultValue=1, group=Main Texture, parameterName=Brightness, sortPriority=4}}
  MaterialExpressionScalarParameter_20 {class=MaterialExpressionScalarParameter, properties {defaultValue=3, group=SecondaryMask, parameterName=StripMaskPower}}
  MaterialExpressionScalarParameter_3 {class=MaterialExpressionScalarParameter, properties {defaultValue=1, group=SecondaryMask, parameterName=SecondaryDarkness, sortPriority=0}}
  MaterialExpressionScalarParameter_4 {class=MaterialExpressionScalarParameter, properties {defaultValue=1, group=SecondaryMask, parameterName=SecondaryBrightness, sortPriority=2}}
  MaterialExpressionScalarParameter_5 {
   class=MaterialExpressionScalarParameter
   properties {
    controlType=Enumeration
    enumeration=&Enum_DLR_TrueFalse
    group=SecondaryMask
    parameterName=UseStripsMask?
    sortPriority=3
   }
  }
  MaterialExpressionScalarParameter_6 {
   class=MaterialExpressionScalarParameter
   properties {
    controlType=Enumeration
    enumeration=&Enum_DLR_TrueFalse
    group=SecondaryMask
    parameterName=UseSecondaryMask?
    sortPriority=5
   }
  }
  MaterialExpressionScalarParameter_7 {
   class=MaterialExpressionScalarParameter
   properties {defaultValue=0.08, group=Distortion OpacityMask, parameterName=AlphaDistortion Strength}
  }
  MaterialExpressionScalarParameter_8 {class=MaterialExpressionScalarParameter, properties {defaultValue=1, group=OpacityMask, parameterName=AlphaMaskDarkness, sortPriority=3}}
  MaterialExpressionScalarParameter_9 {class=MaterialExpressionScalarParameter, properties {defaultValue=1, group=OpacityMask, parameterName=AlphaMaskBrightness, sortPriority=4}}
  MaterialExpressionSine_0 {class=MaterialExpressionSine, inputs [None {from=MaterialExpressionComponentMask_1}], properties {period=2}}
  MaterialExpressionTextureCoordinate_0 {class=MaterialExpressionTextureCoordinate}
  MaterialExpressionTextureCoordinate_1 {class=MaterialExpressionTextureCoordinate}
  MaterialExpressionTextureCoordinate_2 {class=MaterialExpressionTextureCoordinate}
  MaterialExpressionTextureCoordinate_3 {class=MaterialExpressionTextureCoordinate}
  MaterialExpressionTextureCoordinate_4 {class=MaterialExpressionTextureCoordinate}
  MaterialExpressionTextureCoordinate_5 {class=MaterialExpressionTextureCoordinate}
  MaterialExpressionTextureSampleParameter2D_0 {
   class=MaterialExpressionTextureSampleParameter2D
   inputs [
    UVs {from=MaterialExpressionPanner_0, output=""}
    Apply View MipBias {from="", output=""}
   ]
   properties {group=Distortion, parameterName=Distort Texture, sortPriority=2, texture=/Game/Effects/Fort_Effects/Textures/Tile/T_PerlinRGB_01_M}
  }
  MaterialExpressionTextureSampleParameter2D_1 {
   class=MaterialExpressionTextureSampleParameter2D
   inputs [
    UVs {from=MaterialExpressionPanner_1, output=""}
    Apply View MipBias {from="", output=""}
   ]
   properties {group=Main Texture, parameterName=Main Texture, sortPriority=2, texture=/DealershipSharedLibrary/FX/Textures/Noises/T_Energy}
  }
  MaterialExpressionTextureSampleParameter2D_2 {
   class=MaterialExpressionTextureSampleParameter2D
   inputs [
    UVs {from=MaterialExpressionPanner_3, output=""}
    Apply View MipBias {from="", output=""}
   ]
   properties {
    group=Distortion OpacityMask
    parameterName=AlphaDistort Texture
    sortPriority=2
    texture=&T_PerlinRGB_01_M
   }
  }
  MaterialExpressionTextureSampleParameter2D_3 {
   class=MaterialExpressionTextureSampleParameter2D
   inputs [
    UVs {from=MaterialExpressionPanner_4, output=""}
    Apply View MipBias {from="", output=""}
   ]
   properties {
    group=OpacityMask
    parameterName=AlphaMask Texture
    sortPriority=2
    texture=&/T_Infinite_MaskBack_A
   }
  }
  MaterialExpressionTextureSampleParameter2D_4 {
   class=MaterialExpressionTextureSampleParameter2D
   inputs [
    UVs {from=MaterialExpressionPanner_5, output=""}
    Apply View MipBias {from="", output=""}
   ]
   properties {
    group=Clouds Texture
    parameterName=CloudTexture
    samplerType=SAMPLERTYPE_LinearColor
    sortPriority=2
    texture=/Game/Effects/FogVolume/noise_flamy_Tile_Linear
   }
  }
  MaterialExpressionVectorParameter_0 {
   class=MaterialExpressionVectorParameter
   properties {
    group=Main Texture
    parameterName=UVs MainTexture
    sortPriority=0
    channelNames {a=Speed Y, b=Speed X, g=Tiling Y, r=Tiling X}
    defaultValue=<rgba 1 1 0 0>
   }
  }
  MaterialExpressionVectorParameter_1 {
   class=MaterialExpressionVectorParameter
   properties {
    group=Distortion
    parameterName=UVs DistortionTexture
    sortPriority=0
    channelNames {a=Speed Y, b=Speed X, g=Tiling Y, r=Tiling X}
    defaultValue=<rgba 1 1 0 0>
   }
  }
  MaterialExpressionVectorParameter_2 {
   class=MaterialExpressionVectorParameter
   properties {
    group=Distortion OpacityMask
    parameterName=UVs AlphaMaskDistortionTexture
    sortPriority=0
    channelNames {a=Speed Y, b=Speed X, g=Tiling Y, r=Tiling X}
    defaultValue=<rgba 1 1 0 0>
   }
  }
  MaterialExpressionVectorParameter_3 {
   class=MaterialExpressionVectorParameter
   properties {
    group=OpacityMask
    parameterName=UVs AlphaMaskTexture
    sortPriority=1
    channelNames {a=Speed Y, b=Speed X, g=Tiling Y, r=Tiling X}
    defaultValue=<rgba 1 1 0 0>
   }
  }
  MaterialExpressionVectorParameter_4 {
   class=MaterialExpressionVectorParameter
   properties {
    group=Clouds Texture
    parameterName=UVs CloudsTexture
    sortPriority=1
    channelNames {a=Speed Y, b=Speed X, g=Tiling Y, r=Tiling X}
    defaultValue=<rgba 1 1 0 0>
   }
  }
 }
 properties {bUsedWithNiagaraMeshParticles=true, blendMode=BLEND_Masked, shadingModel=MSM_Unlit, usedShadingModels=Unlit}
}
